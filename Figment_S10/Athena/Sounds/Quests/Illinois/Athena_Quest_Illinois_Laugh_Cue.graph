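# Sound cue for the Illinois quest that plays a balloon pop followed shortly by a laugh.
# A mixer layers a random pick of five pitch-varied balloon pops with a random pick of four laughs, the laughs delayed half a second and slightly modulated.
# Each branch overrides its sound class, ducking impact class for pops and mission in-game class for laughs.
# Attenuation and concurrency are overridden to limit range and simultaneous instances.
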SoundCue /Figment_S10/Athena/Sounds/Quests/Illinois/Athena_Quest_Illinois_Laugh_Cue
root=Mixer_0
nodes [
 {id=Delay_0, children [Random_1], properties {delayMax=0.5, delayMin=0.5}}
 {id=Mixer_0, children [SoundClass_1,SoundClass_0], properties {inputVolume [1,1]}}
 {id=Modulator_0, children [Delay_0], properties {pitchMax=1, pitchMin=0.99, volumeMax=1, volumeMin=0.9}}
 {id=Modulator_1, children [Random_0], properties {pitchMax=1.1, pitchMin=0.9}}
 {id=Random_0, children [WavePlayer_0,WavePlayer_2,WavePlayer_3,WavePlayer_4,WavePlayer_1], properties {weights [1,1,1,1,1]}}
 {id=Random_1, children [WavePlayer_7, WavePlayer_8, WavePlayer_9, WavePlayer_10], properties {weights [1,1,1,1]}}
 {
  id=SoundClass_0
  children [Modulator_1]
  properties {soundClassOverride=/Game/Sounds/FortSoundClassesAndModes/Fort_ImpactSounds_DuckDefaultQuieter_PSM}
 }
 {id=SoundClass_1, children [Modulator_0], properties {soundClassOverride=/Game/Sounds/FortSoundClassesAndModes/Fort_Mission_IngameSounds}}
 {id=WavePlayer_0, properties {soundWaveAssetPtr=/Game/Sounds/Fort_Gadgets/Balloons/Balloon_Pop_01}}
 {id=WavePlayer_1, properties {soundWaveAssetPtr=/Game/Sounds/Fort_Gadgets/Balloons/Balloon_Pop_05}}
 {id=WavePlayer_10, properties {soundWaveAssetPtr=&/Illinois_Laugh_04}}
 {id=WavePlayer_2, properties {soundWaveAssetPtr=/Game/Sounds/Fort_Gadgets/Balloons/Balloon_Pop_02}}
 {id=WavePlayer_3, properties {soundWaveAssetPtr=/Game/Sounds/Fort_Gadgets/Balloons/Balloon_Pop_03}}
 {id=WavePlayer_4, properties {soundWaveAssetPtr=/Game/Sounds/Fort_Gadgets/Balloons/Balloon_Pop_04}}
 {id=WavePlayer_7, properties {soundWaveAssetPtr=&/Illinois_Laugh_01}}
 {id=WavePlayer_8, properties {soundWaveAssetPtr=&/Illinois_Laugh_02}}
 {id=WavePlayer_9, properties {soundWaveAssetPtr=&/Illinois_Laugh_03}}
]
properties {
 attenuationSettings=&Fort_GadgetActivate_Att
 bOverrideAttenuation=true
 bOverrideConcurrency=true
 maxDistance=2000
 volumeMultiplier=1.2
 attenuationOverrides {falloffDistance=2000, attenuationShapeExtents {x=0}}
 concurrencyOverrides {maxCount=3}
}
